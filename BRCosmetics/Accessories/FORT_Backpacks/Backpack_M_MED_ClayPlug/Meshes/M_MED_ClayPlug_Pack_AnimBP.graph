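# Animation Blueprint for the M_MED_ClayPlug back bling, handling its pack placement, dangling wire and gauge needle.
# AnimDynamics swings base_wire, then stacked additive root and pack offsets feed a cached pose, which branches into two needle modifiers blended by Radioactive.
# Each update derives spine-based pack offsets, flags for matching character mesh, front end, skydiving and melee, plus sine-driven needle rotators.
# Radioactive turns on during movement or actions like aiming and melee.
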AnimBlueprint /BRCosmetics/Accessories/FORT_Backpacks/Backpack_M_MED_ClayPlug/Meshes/M_MED_ClayPlug_Pack_AnimBP
target_skeleton=&/M_MED_ClayPlug_Pack_Skeleton
graphs [
 AnimGraph {nodes [
   "Transform (Modify) Bone - Bone: needle"(Alpha=1, Translation="0,0,0", Rotation="0,0,0", Scale="1,1,1") {
    rotationMode=BMM_Additive
    rotationSpace=BCS_BoneSpace
    translationMode=BMM_Additive
    translationSpace=BCS_BoneSpace
    alphaBoolBlend {blendOption=Linear}
    boneToModify {boneName=needle}
   }
   "Transform (Modify) Bone - Bone: needle"(Alpha=0, Translation="0,0,0", Rotation="0,0,0", Scale="1,1,1") {
    rotationMode=BMM_Additive
    rotationSpace=BCS_BoneSpace
    translationMode=BMM_Additive
    translationSpace=BCS_BoneSpace
    alphaBoolBlend {blendOption=Linear}
    boneToModify {boneName=needle}
   }
   n1 = "Transform (Modify) Bone - Bone: needle"(
    ComponentPose=Local To Component(LocalPose=Use cached pose 'FixedPositionCache'())
    Alpha=1
    Translation="0,0,0"
    Rotation="0,0,0"
    Scale="1,1,1"
   ) {
    rotationMode=BMM_Additive
    rotationSpace=BCS_BoneSpace
    translationMode=BMM_Additive
    translationSpace=BCS_BoneSpace
    alphaBoolBlend {blendOption=Linear}
    boneToModify {boneName=needle}
   }
   n2 = "Transform (Modify) Bone - Bone: needle"(
    ComponentPose=Local To Component(LocalPose=Use cached pose 'FixedPositionCache'())
    Alpha=1
    Translation="0,0,0"
    Rotation="0,0,0"
    Scale="1,1,1"
   ) {
    rotationMode=BMM_Additive
    rotationSpace=BCS_BoneSpace
    translationMode=BMM_Additive
    translationSpace=BCS_BoneSpace
    alphaBoolBlend {blendOption=Linear}
    boneToModify {boneName=needle}
   }
   Output Pose(Result=Blend Poses by bool(
     BlendPose_0=Component To Local(ComponentPose=@n2)
     BlendPose_1=Component To Local(ComponentPose=@n1)
     BlendTime_0=0.5
     BlendTime_1=0.5
     bActiveValue=Get Radioactive() -> (Radioactive:bool)
    ) {blendType=Linear}) {layerGroup=None}
   n3 = Anim Dynamics - Bone: base_wire!(SimulationSpace:AnimPhysSimSpaceType Enum=World, ComponentPose=Local To Component(LocalPose=Input Pose()), Alpha=1, ExternalForce="0,0,0") -> (Pose) {
    angularDampingOverride=1
    angularSpringConstant=25000
    bAngularSpring=true
    bOverrideAngularDamping=true
    bUseGravityOverride=true
    gravityScale=0.35
    simulationSpace=World
    alphaBoolBlend {blendOption=Linear}
    boundBone {boneName=base_wire}
    componentLinearAccScale=<1 1 1>
    componentLinearVelScale=<1 1 1>
    gravityOverride {z=-980}
    physicsBodyDefinitions [{
      boundBone {boneName=base_wire}
      boxExtents=<20 30 30>
      constraintSetup {angularLimitsMax=<0 15 25>, angularLimitsMin=<0 -15 -25>, angularTarget=<1 0 0>}
      localJointOffset=<4 0 0>
     }]
   }
   n4 = "Transform (Modify) Bone - Bone: pack"(ComponentPose=@n3, Translation="0,-1.9505,0", Rotation="0,0,0") {
    translationMode=BMM_Additive
    translationSpace=BCS_BoneSpace
    alphaBoolBlend {blendOption=Linear}
    boneToModify {boneName=pack}
    translation {y=-1.9505}
   }
   n5 = "Transform (Modify) Bone - Bone: root"(ComponentPose=@n4, Translation="0.000,3.100,-6.640", Rotation="-0.000,0.000,5.000") {
    rotationMode=BMM_Additive
    rotationSpace=BCS_BoneSpace
    translationMode=BMM_Additive
    translationSpace=BCS_BoneSpace
    alphaBoolBlend {blendOption=Linear}
    boneToModify {boneName=root}
    rotation {roll=5}
    translation {y=3.1, z=-6.64}
   }
   n6 = "Transform (Modify) Bone - Bone: root"(ComponentPose=@n5, Translation="0,1.3,0", Rotation="0,0,5") {
    rotationMode=BMM_Additive
    rotationSpace=BCS_BoneSpace
    translationMode=BMM_Additive
    translationSpace=BCS_BoneSpace
    alphaBoolBlend {blendOption=Linear}
    boneToModify {boneName=root}
    rotation {roll=5}
    translation {y=1.3}
   }
   n7 = "Transform (Modify) Bone - Bone: root"(ComponentPose=@n6, Translation="0,0.6,0", Rotation="0,0,5") {
    rotationMode=BMM_Additive
    rotationSpace=BCS_BoneSpace
    translationMode=BMM_Additive
    translationSpace=BCS_BoneSpace
    alphaBoolBlend {blendOption=Linear}
    boneToModify {boneName=root}
    rotation {roll=5}
    translation {y=0.6}
   }
   n8 = "Transform (Modify) Bone - Bone: pack"!(ComponentPose=@n7, Alpha, Translation="0,0,0", Rotation="0,0,0") -> (Pose) {rotationMode=BMM_Additive, translationMode=BMM_Additive, alphaBoolBlend {blendOption=Linear}, boneToModify {boneName=pack}}
   n9 = "Transform (Modify) Bone - Bone: pack"!(ComponentPose=@n8, Alpha, Translation="0.000,-4.292,1.603", Rotation="-0.000,0.000,28.440", Scale:Vector) -> (Pose) {
    rotationMode=BMM_Additive
    rotationSpace=BCS_BoneSpace
    translationMode=BMM_Additive
    translationSpace=BCS_BoneSpace
    alphaBoolBlend {blendOption=Linear}
    alphaScaleBiasClamp {bInterpResult=true, interpSpeedDecreasing=3, interpSpeedIncreasing=5}
    boneToModify {boneName=pack}
    rotation {roll=28.44}
    translation {y=-4.292, z=1.603}
   }
   n10 = "Transform (Modify) Bone - Bone: pack"(ComponentPose=@n9, Translation="-0.000,-0.958,0.143", Rotation="-0.000,0.000,8.696") {
    rotationMode=BMM_Additive
    rotationSpace=BCS_BoneSpace
    translationMode=BMM_Additive
    translationSpace=BCS_BoneSpace
    alphaBoolBlend {blendOption=Linear}
    alphaScaleBiasClamp {bInterpResult=true, interpSpeedDecreasing=3, interpSpeedIncreasing=4}
    boneToModify {boneName=pack}
    rotation {roll=8.696}
    translation=<0 -0.9576 0.1429>
   }
   n11 = "Transform (Modify) Bone - Bone: pack"!(ComponentPose=@n10, bAlphaBoolEnabled:bool, Translation="0,-1.5,0", Rotation="0,0,0", Scale) -> (Pose) {
    alphaInputType=Bool
    rotationMode=BMM_Additive
    rotationSpace=BCS_BoneSpace
    translationMode=BMM_Additive
    translationSpace=BCS_BoneSpace
    alphaBoolBlend {blendInTime=0.2, blendOption=Linear, blendOutTime=0.2}
    alphaScaleBiasClamp {bInterpResult=true, interpSpeedDecreasing=3, interpSpeedIncreasing=6}
    boneToModify {boneName=pack}
    translation {y=-1.5}
   }
   n12 = "Transform (Modify) Bone - Bone: pack"!(ComponentPose=@n11, bAlphaBoolEnabled, Translation="0,0.3,0", Rotation="0,0,3") -> (Pose) {
    alphaInputType=Bool
    rotationMode=BMM_Additive
    translationMode=BMM_Additive
    alphaBoolBlend {blendInTime=0.2, blendOption=Linear, blendOutTime=0.2}
    boneToModify {boneName=pack}
    rotation {roll=3}
    translation {y=0.3}
   }
   n13 = "Transform (Modify) Bone - Bone: pack"(ComponentPose=@n12, Translation="0,0.2,0", Rotation="0,0,0") {
    alphaInputType=Bool
    rotationMode=BMM_Additive
    translationMode=BMM_Additive
    alphaBoolBlend {blendInTime=0.2, blendOption=Linear, blendOutTime=0.2}
    boneToModify {boneName=pack}
    translation {y=0.2}
   }
   Save cached pose 'FixedPositionCache'(Pose=Component To Local(ComponentPose=@n13)) {cacheName=FixedPositionCache}
  ]}
 EventGraph {nodes [
   Event BlueprintUpdateAnimation()
   > n1 = Sequence!(execute) -> (then_0, then_1, then_2:Exec, then_3:Exec, then_4:Exec, then_5:Exec, then_6:Exec, then_7:Exec, then_8:Exec, then_9:Exec)
   n2 = Cast To FortPlayerPawn(execute=@n1.then_2, Object=TryGetPawnOwner())
   n3 = MatchesWildcard(SourceString=GetDisplayName(Object=GetSkeletalMeshForPartType(self=@n2)), Wildcard=*M_MED_ClayPlug*, SearchCase=CaseSensitive)
   > n4 = Branch(Condition=@n3)
   > Set Position Backpack to MainCharacter(Position Backpack to MainCharacter:double=1) -> (Output_Get:double)
   > Set Position Backpack to OtherCharacter(Position Backpack to OtherCharacter:double=0) -> (Output_Get:double)
   Set Position Backpack to MainCharacter(execute=@n4.else, Position Backpack to MainCharacter=0)
   > Set Position Backpack to OtherCharacter(Position Backpack to OtherCharacter=1)
   n5 = "Sin (Radians)"(A=float * float(A=float * float(A=0.5, B=45), B=GetRealTimeSeconds()))
   Set dyn_needle_Fast_rotator(execute=@n1.then_5, dyn_needle_Fast_rotator:Rotator=MakeRotator(Pitch=float * float(A=@n5, B=-50))) -> (Output_Get:Rotator)
   n6 = Branch(execute=@n1.then_7, Condition=NOT Boolean(A=Get bIsPlayingMeleeAnim()))
   > Set AlphaIsNotMelee(AlphaIsNotMelee:double=1) -> (Output_Get:double)
   Set AlphaIsNotMelee(execute=@n6.else, AlphaIsNotMelee=0)
   n7 = Sequence!(execute=@n1.then_0) -> (then_0, then_1)
   n8 = GetOwnerSkeletalMeshComponent()
   n9 = BreakRotator(InRot=BreakTransform(InTransform=GetDeltaTransformFromRefPose(self=@n8, BoneName=spine_05, BaseName=spine_01)).Rotation)
   > Set Backpack Rotate(Backpack Rotate:Rotator=MakeRotator(Roll=MapRangeClamped(Value=@n9.Yaw, InRangeA=0, InRangeB=-80, OutRangeB=20))) -> (Output_Get:Rotator)
   Set Backpack Translation(execute=@n7.then_1, Backpack Translation:Vector=MakeVector(Y=MapRangeClamped(Value=@n9.Yaw, InRangeA=0, InRangeB=-30, OutRangeB=-1.9))) -> (Output_Get:Vector)
   n10 = Get bIsSkyDiving()
   n11 = Get bIsMoving2D()
   n12 = Get bIsJumping()
   n13 = Branch(execute=@n1.then_8, Condition=AND Boolean(A=OR Boolean(A=@n11, B=@n12), B=NOT Boolean(A=@n10)))
   > Set JumpMoving2D(JumpMoving2D:double=1) -> (Output_Get:double)
   Set JumpMoving2D(execute=@n13.else, JumpMoving2D=0)
   n14 = Branch(execute=@n1.then_3, Condition=AND Boolean(A=NOT Boolean(A=@n3), B=Get bIsInFrontEnd()))
   > Set FrontEnd Distance(FrontEnd Distance:double=1) -> (Output_Get:double)
   Set FrontEnd Distance(execute=@n14.else, FrontEnd Distance=0)
   n15 = Branch(execute=@n1.then_9, Condition=Get bIsInFrontEnd())
   > Set SimlationSpace(SimlationSpace:AnimPhysSimSpaceType Enum=World) -> (Output_Get:AnimPhysSimSpaceType Enum)
   Set SimlationSpace(execute=@n15.else, SimlationSpace=Component)
   n16 = Branch(
    execute=@n1.then_6
    Condition=OR Boolean!(
     A=Get bIsMoving2D()
     B=Get bIsTargeting()
     C:bool=Get bIsSkyDiving()
     D:bool=Get bIsSliding()
     E:bool=Get bIsJumping()
     F:bool=Get bIsFalling()
     G:bool=Get bIsPlayingMeleeAnim()
    ) -> (ReturnValue)
   )
   > Set Radioactive(Radioactive:bool=true) -> (Output_Get:bool)
   Set Radioactive(execute=@n16.else, Radioactive=false)
   n17 = GetOwnerSkeletalMeshComponent()
   n18 = BreakRotator(InRot=BreakTransform(InTransform=GetDeltaTransformFromRefPose(self=@n17, BoneName=pelvis, BaseName=root)).Rotation)
   n19 = Branch(
    execute=@n1.then_1
    Condition=AND Boolean!(A="float > float"(A=@n18.Yaw, B=90), B="float < float"(A=@n18.Yaw, B=130), C:bool=Get bIsSkydiveDiveMode()) -> (ReturnValue)
   )
   > Set AlphaIsCorrectiveSkydiveMode(AlphaIsCorrectiveSkydiveMode:double=1) -> (Output_Get:double)
   Set AlphaIsCorrectiveSkydiveMode(execute=@n19.else, AlphaIsCorrectiveSkydiveMode=0)
   n20 = "Sin (Radians)"(A=float * float(A=float * float(A=0.5, B=55), B=GetRealTimeSeconds()))
   Set!(
    execute=@n1.then_4
    GaugeRotation_Roll:float=0
    GaugeRotation_Pitch:float=MakeRotator!(Roll, Pitch=float * float(A=@n20, B=-8), Yaw) -> (ReturnValue_Roll:float, ReturnValue_Pitch:float, ReturnValue_Yaw:float).ReturnValue_Pitch
    GaugeRotation_Yaw:float=0
   ) -> (then, Output_Get:Rotator)
   > Delay(Duration=RandomFloatInRange(Min=0.5, Max=0))
   > Set!(execute, Spinning_Rotator_Roll:float=0, Spinning_Rotator_Pitch:float=RandomFloatInRange(Min=20, Max=0), Spinning_Rotator_Yaw:float=0) -> (then, Output_Get)
  ]}
]
variables [
 AlphaIsCorrectiveSkydiveMode {type=number}
 AlphaIsNotMelee {type=number}
 Backpack Rotate {type=Rotator}
 Backpack Translation {type=Vector}
 dyn_needle_Fast_rotator {type=Rotator, default=<rot 23.4354 0 0>}
 dyn_needle_Slow_rotator {type=Rotator, default=<rot 23.4354 0 0>}
 Forward {type=boolean}
 FrontEnd Distance {type=number}
 GaugeRotation {type=Rotator}
 Is Jumping {type=boolean}
 JumpMoving2D {type=number}
 Position Backpack to MainCharacter {type=number}
 Position Backpack to OtherCharacter {type=number}
 Radioactive {type=boolean}
 SimlationSpace {type=AnimPhysSimSpaceType}
 Spinning_Rotator {type=Rotator}
]
